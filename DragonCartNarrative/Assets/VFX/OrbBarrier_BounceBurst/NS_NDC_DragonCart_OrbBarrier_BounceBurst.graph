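# Niagara one-shot orange burst effect, named for a bounce off the DragonCart orb barrier.
# Burst layers timed spawn groups in one emitter, using SpawnGroup switches for lifetime, size and fade, and VisibilityTag for its four sprite renderers.
# Sparks throws a cone of speed-stretched sprites that fall under drag and gravity.
# User.Scale scales sprite sizes and the scatter radius; two mesh emitters are disabled.
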NiagaraSystem /DragonCartNarrative/Assets/VFX/OrbBarrier_BounceBurst/NS_NDC_DragonCart_OrbBarrier_BounceBurst
dependencies {
 usedDataInterfaces [
  &NiagaraDataInterfaceCurve
  &NiagaraDataInterfaceVector2DCurve
  &NiagaraDataInterfaceVectorCurve
 ]
 usedDynamicInputs [&RandomVector]
 usedModules [&SystemState]
}
emitters {
 Burst {
  emitterUpdateScript [
   &EmitterState()
   &SpawnBurst_Instantaneous(Use Spawn Probability=false, UseLoopCountLimit=false)
   &SpawnBurst_Instantaneous(Spawn Group=1, Spawn Time=0.001, Use Spawn Probability=false, UseLoopCountLimit=false) {module=SpawnBurst_Instantaneous001}
   &SpawnBurst_Instantaneous(Spawn Group=3, Use Spawn Probability=false, UseLoopCountLimit=false) {module=SpawnBurst_Instantaneous003}
   &SpawnBurst_Instantaneous(Spawn Count=10, Spawn Group=4, Spawn Time=0.04, Use Spawn Probability=false, UseLoopCountLimit=false) {module=SpawnBurst_Instantaneous004}
   &SpawnBurst_Instantaneous(Spawn Count=3, Spawn Group=4, Spawn Time=0.2, Use Spawn Probability=false, UseLoopCountLimit=false) {module=SpawnBurst_Instantaneous002}
   &SpawnBurst_Instantaneous(Spawn Count=4, Spawn Group=4, Spawn Time=0.3, Use Spawn Probability=false, UseLoopCountLimit=false) {module=SpawnBurst_Instantaneous005}
  ]
  particleSpawnScript [
   /SetVariables(Particles.VisibilityTag=&SpawnGroup(Max Spawn Group=5)) {isSetParameters=true}
   &InitializeParticle(
    Color=<rgba 5 1.4112 0 1>
    Lifetime=&MergeHelper_FloatSwitch(
     0=&RandomRangeFloat(Minimum=0.15, Maximum=0.19)
     1=0.4
     2=0
     3=0.4
     4=&RandomRangeFloat(Minimum=0.1, Maximum=0.4)
     Index=&SpawnGroup()
     Num=::NewEnumerator3
    )
    Mesh Scale=<0.3 0.3 0.3>
    Mesh Scale Mode=::NewEnumerator4
    Position=$Particles.Position
    Position Mode=::NewEnumerator1
    Sprite Size Mode=::NewEnumerator3
    Uniform Sprite Size=&MergeHelper_FloatSwitch(
     0=300
     1=1200
     2=0
     3=900
     4=&RandomRangeFloat(Minimum=55, Maximum=67)
     Index=&SpawnGroup()
     Num=::NewEnumerator3
    )
    UsePositionOffset=false
    Write Material Random=true
   )
   &CameraOffset(Camera Offset Amount=20)
   /SetVariables(Particles.RunesDynamic:Vector4f=&MakeVector4(
     X=&RandomRangeFloat(Minimum=0, Maximum=1)
     Y=&RandomRangeFloat(Minimum=1, Maximum=12)
     Z=1
     W=1
    )) {isSetParameters=true}
   &SpriteFacingAndAlignment(WriteAlignment=true, WriteFacing=true)
   &SystemLocation(MaskSpawnGroup=false, Offset=<-30 0 -120>, Offset Coordinate Space=::Simulation)
   &ShapeLocation(
    Disc Coverage=1
    MaskSpawnGroup=false
    Override Seed=false
    Ring Radius=&MergeHelper_FloatSwitch(
     0=0
     1=0
     2=0
     3=0
     4=&Multiply_Float(A=200, B=$User.Scale)
     Index=&SpawnGroup()
     Num=::NewEnumerator3
    )
    Rotation Angle Type:ENiagara_AngleInput=::NewEnumerator0
    Rotation Coordinate Space=::Simulation
    Rotation Mode=::NewEnumerator1
    Scale Mode:ENiagara_ScaleMode=::NewEnumerator1
    Shape Primitive=::NewEnumerator4
   )
  ]
  particleUpdateScript [
   &ParticleState()
   &ScaleSpriteSize(
    Uniform Curve Scale=$User.Scale
    Uniform Curve Sprite Scale=DataInterface() {propertyValues {Curve {keys [
        {arriveTangent=1.8511, interpMode=RCIM_Cubic, leaveTangent=1.8511, tangentMode=RCTM_User, time=0, value=0.2}
        {tangentMode=RCTM_User, time=1, value=1}
       ]}}}
   )
   &ScaleMeshSize(Scale Factor=&VectorFromFloat(Value=&FloatFromCurve(
      FloatCurve=DataInterface() {propertyValues {Curve {keys [
          {arriveTangent=13.1394, interpMode=RCIM_Cubic, leaveTangent=13.1394, tangentMode=RCTM_User, time=0, value=0.1}
          {arriveTangent=5.1695, interpMode=RCIM_Cubic, leaveTangent=5.1695, tangentMode=RCTM_User, time=1, value=8}
         ]}}}
      Scale Curve=$User.Scale
     ))) {enabled=false}
   &SolveForcesAndVelocity(Clamp Velocity=false, Limit Acceleration=false)
   &ScaleColor(
    Scale Alpha=&MergeHelper_FloatSwitch(
     0=&FloatFromCurve(FloatCurve=DataInterface() {propertyValues {Curve {
         defaultValue=1
         keys [
          {arriveTangent=-2.3974, interpMode=RCIM_Cubic, leaveTangent=-2.3974, tangentMode=RCTM_User, time=0, value=1}
          {interpMode=RCIM_Cubic, tangentMode=RCTM_User, time=1, value=0}
         ]
        }}})
     1=&FloatFromCurve(FloatCurve=DataInterface() {propertyValues {Curve {
         defaultValue=1
         keys [
          {arriveTangent=9.0725, interpMode=RCIM_Cubic, leaveTangent=9.0725, tangentMode=RCTM_User, time=0, value=0}
          {arriveTangent=-2.5249, interpMode=RCIM_Cubic, leaveTangent=-2.5249, tangentMode=RCTM_User, time=0.1408, value=0.9847}
          {arriveTangent=-0.0405, interpMode=RCIM_Cubic, leaveTangent=-0.0405, tangentMode=RCTM_User, time=1, value=0}
         ]
        }}})
     2=&FloatFromCurve(FloatCurve=DataInterface() {propertyValues {Curve {keys [{arriveTangent=-1.8208, interpMode=RCIM_Cubic, leaveTangent=-1.8208, tangentMode=RCTM_User, time=0, value=1}, <at 1 0>]}}})
     3=&FloatFromCurve(
      FloatCurve=DataInterface() {propertyValues {Curve {
         defaultValue=1
         keys [
          {interpMode=RCIM_Cubic, tangentMode=RCTM_User, time=0, value=1}
          {arriveTangent=-0.8798, interpMode=RCIM_Cubic, leaveTangent=-0.8798, time=0.2069, value=2}
          {arriveTangent=-2.5216, interpMode=RCIM_Cubic, leaveTangent=-2.5216, time=0.3968, value=0.6509}
          {interpMode=RCIM_Cubic, tangentMode=RCTM_User, time=1, value=0}
         ]
        }}}
      Scale Curve=0.6
     )
     4=&FloatFromCurve(FloatCurve=DataInterface() {propertyValues {Curve {
         defaultValue=1
         keys [
          {interpMode=RCIM_Cubic, tangentMode=RCTM_User, time=0, value=1}
          {arriveTangent=0.0104, interpMode=RCIM_Cubic, leaveTangent=0.0104, tangentMode=RCTM_User, time=1, value=0}
         ]
        }}})
     Index=&SpawnGroup()
     Num=::NewEnumerator3
    )
    Scale RGB=&VectorFromFloat(Value=&MergeHelper_FloatSwitch(
      0=1
      1=1
      2=0
      3=1
      4=&FloatFromCurve(FloatCurve=DataInterface() {propertyValues {Curve {
          defaultValue=1
          keys [
           {arriveTangent=9.0725, interpMode=RCIM_Cubic, leaveTangent=9.0725, tangentMode=RCTM_User, time=0, value=1}
           {arriveTangent=-1.2769, interpMode=RCIM_Cubic, leaveTangent=-1.2769, tangentMode=RCTM_User, time=0.1408, value=5}
           {arriveTangent=-0.9788, interpMode=RCIM_Cubic, leaveTangent=-0.9788, tangentMode=RCTM_User, time=1, value=1}
          ]
         }}})
      Index=&SpawnGroup()
      Num=::NewEnumerator3
     ))
    ScaleA=true
    ScaleRGB=true
   )
  ]
  rendererClasses [&NiagaraSpriteRendererProperties]
  renderers [
   {rendererClass=&NiagaraSpriteRendererProperties}
   {rendererClass=&NiagaraSpriteRendererProperties, rendererIndex=1}
   {rendererClass=&NiagaraSpriteRendererProperties, rendererIndex=2}
   {rendererClass=&NiagaraSpriteRendererProperties, rendererIndex=3}
  ]
 }
 MeshOrientation {
  bEnabled=false
  emitterUpdateScript [
   &EmitterState()
   &SpawnBurst_Instantaneous(Use Spawn Probability=false, UseLoopCountLimit=false)
  ]
  particleSpawnScript [
   &InitializeParticle(
    Lifetime=3
    Mesh Scale Mode=::NewEnumerator1
    Mesh Uniform Scale Max=5
    Mesh Uniform Scale Min=4
    UsePositionOffset=false
    Write Material Random=true
   )
   &InitialMeshOrientation(Mesh Orientation Mode=::NewEnumerator1, Use Rotation Vector=false)
  ]
  particleUpdateScript [
   &ParticleState()
   &UpdateMeshOrientation(Rotation Vector=<0 0 0.5>) {enabled=false}
   &ScaleMeshSize(Scale Factor=&VectorFromFloat(Value=&FloatFromCurve(FloatCurve=DataInterface() {propertyValues {Curve {keys [<at 0.65 1>,<at 1 0>]}}}))) {enabled=false}
  ]
  rendererClasses [&NiagaraMeshRendererProperties]
  renderers [{rendererClass=&NiagaraMeshRendererProperties, rendererIndex=0}]
 }
 MeshOrientation001 {
  bEnabled=false
  emitterUpdateScript [
   &EmitterState()
   &SpawnBurst_Instantaneous(Use Spawn Probability=false, UseLoopCountLimit=false)
  ]
  particleSpawnScript [
   &InitializeParticle(
    Lifetime=3
    Mesh Scale Mode=::NewEnumerator1
    Mesh Uniform Scale Max=5
    Mesh Uniform Scale Min=4
    UsePositionOffset=false
    Write Material Random=true
   )
   &InitialMeshOrientation(Mesh Orientation Mode=::NewEnumerator1, Use Rotation Vector=false)
   &SystemLocation(MaskSpawnGroup=false, Offset=<-30 0 -120>, Offset Coordinate Space=::Simulation)
  ]
  particleUpdateScript [
   &ParticleState()
   &UpdateMeshOrientation(Rotation Vector=<0 0 0.5>) {enabled=false}
   &ScaleMeshSize(Scale Factor=&VectorFromFloat(Value=&FloatFromCurve(FloatCurve=DataInterface() {propertyValues {Curve {keys [<at 0.65 1>,<at 1 0>]}}}))) {enabled=false}
  ]
  rendererClasses [&NiagaraMeshRendererProperties]
  renderers [{rendererClass=&NiagaraMeshRendererProperties, rendererIndex=0}]
 }
 Sparks {
  emitterUpdateScript [
   &EmitterState()
   &SpawnBurst_Instantaneous(
    Spawn Count=&RandomRangeInt(Minimum=6, Maximum=8)
    Spawn Time=0.01
    Use Spawn Probability=false
    UseLoopCountLimit=false
   ) {module=SpawnBurst_Instantaneous001}
   &SpawnBurst_Instantaneous(
    Spawn Count=&RandomRangeInt(Minimum=6, Maximum=8)
    Spawn Time=0.04
    Use Spawn Probability=false
    UseLoopCountLimit=false
   )
  ]
  particleSpawnScript [
   &InitializeParticle(
    Color=<rgba 10 2.4793 0.0504 1>
    Lifetime=&RandomRangeFloat(Minimum=0.2, Maximum=0.8, Evaluation Type=::NewEnumerator1)
    Sprite Size=&RandomRangeVector2D(Minimum=<4 14>, Maximum=<10 20>, Evaluation Type=::NewEnumerator1)
    Sprite Size Mode=::NewEnumerator4
    UsePositionOffset=false
    Write Material Random=true
   )
   &AddVelocity(
    Cone Angle=80
    Module.UseConeFalloff=false
    Module.UseVelDistribution=false
    Override Seed=false
    Rotation Coordinate Space=::Simulation
    Velocity Mode=::NewEnumerator2
    Velocity Speed=&RandomRangeFloat(Minimum=400, Maximum=1200)
   )
   &SystemLocation(MaskSpawnGroup=false, Offset=<-30 0 -120>, Offset Coordinate Space=::Simulation)
  ]
  particleUpdateScript [
   &ParticleState()
   &ScaleSpriteSize(
    Scale Factor=&Vector2DFromCurve(
     Vector2Curve=DataInterface() {propertyValues {
       XCurve {keys [
         {interpMode=RCIM_Cubic, tangentMode=RCTM_User, time=0, value=1.5}
         {arriveTangent=-0.5, interpMode=RCIM_Cubic, leaveTangent=-0.5, tangentMode=RCTM_User, time=0.2853, value=1.7346}
         <at 1.0166 0.4393 RCIM_Cubic>
        ]}
       YCurve {keys [
         {arriveTangent=1.7792, interpMode=RCIM_Cubic, leaveTangent=1.7792, tangentMode=RCTM_User, time=0, value=1}
         {arriveTangent=-0.4733, interpMode=RCIM_Cubic, leaveTangent=-0.4733, tangentMode=RCTM_User, time=0.3599, value=1.6016}
         <at 1 0 RCIM_Cubic>
        ]}
      }}
     ScaleCurve=&Multiply_Vector2D_ByFloat(Float=1, Vector2D=&Vector2DFromFloat(Value=$User.Scale))
    )
    Scale Sprite Size Mode=::NewEnumerator2
   )
   &AccelerationForce(Acceleration=&RandomRangeVector(Minimum=<0 0 -900>, Maximum=<0 0 -1650>))
   &Drag(Use Linear Drag=true, Use Rotational Drag=true)
   &GravityForce(Gravity=&RandomRangeVector(Minimum=<0 0 -400>, Maximum=<0 0 -800>))
   &SolveForcesAndVelocity(Clamp Velocity=false, Limit Acceleration=false)
   &ScaleSpriteSizeBySpeed(Max Scale Factor=<2 3>, Source Velocity:Vector3f=&Multiply_Vector(A=<0 0.005 1>, B=$Particles.Velocity))
   &ScaleColor(
    Scale RGB=&VectorFromCurve(VectorCurve=DataInterface() {propertyValues {
       XCurve {keys [<at 0 1 RCIM_Cubic>,<at 1 0.3333>]}
       YCurve {keys [<at 0 1 RCIM_Cubic>,<at 1 0.026>]}
       ZCurve {keys [<at 0 1 RCIM_Cubic>,<at 1 0.1205>]}
      }})
    ScaleA=true
    ScaleRGB=true
   )
  ]
  rendererClasses [&NiagaraSpriteRendererProperties]
  renderers [{rendererClass=&NiagaraSpriteRendererProperties}]
 }
}
summary {
 systemName=NS_NDC_DragonCart_OrbBarrier_BounceBurst
 emitters [
  {emitterName=Burst, rendererClasses [&NiagaraSpriteRendererProperties]}
  {emitterName=Sparks, rendererClasses [&NiagaraSpriteRendererProperties]}
  {bEnabled=false, emitterName=MeshOrientation, rendererClasses [&NiagaraMeshRendererProperties]}
  {bEnabled=false, emitterName=MeshOrientation001, rendererClasses [&NiagaraMeshRendererProperties]}
 ]
 userVariables [User.Scale {type=NiagaraFloat, defaultValue=1, description=""}]
}
